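# List item widget for the TimeOfDayTools sequence scalability editor utility, derived from TA_BaseEditorButtonWidget; named Checkbox but contains no checkbox.
# Its bordered row holds a scale-down text label and an orange clear-icon button with no logic wired here.
# Overrides expose the border and text to the parent, forward background-color validation, and match search strings against the label.
# The data-change event calls the parent, then casts to TA_GenericListItemData without using the result.
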WidgetBlueprint /Game/NeverCook/EditorUtilities/TimeOfDayTools/TODSeqScalability/Widgets/TODSeqScal_ListItemCheckbox
graphs [
 EventGraph {nodes [
   n1 = Event updateAfterDataChange!() -> (OutputDelegate, then, currentListItemData:Object Reference) {is=title K2Node_Event}
   > Parent: updateAfterDataChange!(execute, currentListItemData:Object Reference=@n1.currentListItemData) -> (then) {is=title K2Node_CallParentFunction}
   > Cast To TA_GenericListItemData!(execute, Object=@n1.currentListItemData) -> (then, CastFailed, AsTA Generic List Item Data:TA Generic List Item Data Object Reference) {is=title K2Node_DynamicCast}
  ]}
 FilterBySearchString {nodes [
   n1 = FilterBySearchString!() -> (then, SearchString:String) {is=title K2Node_FunctionEntry}
   > Return Node!(execute, ShouldDisplay:bool=Contains(SearchIn=Get label() -> (label:String), Substring=@n1))
  ]}
 getBackgroundWidget {nodes [
   getBackgroundWidget() {is=title K2Node_FunctionEntry}
   > Return Node!(execute, backgroundWidget:Widget Object Reference=Get borderMain() -> (borderMain:Border Object Reference))
  ]}
 getLabelWidget {nodes [
   getLabelWidget() {is=title K2Node_FunctionEntry}
   > Return Node!(execute, labelWidget:Widget Object Reference=Get txtMain() -> (txtMain:Text Object Reference))
  ]}
 validateBackgroundColor {nodes [
   validateBackgroundColor() {is=title K2Node_FunctionEntry}
   > n1 = Parent: validateBackgroundColor!(execute) -> (then, handled:bool) {is=title K2Node_CallParentFunction}
   > Return Node!(execute, handled:bool=@n1)
  ]}
]
variables []
widgets {
 info {
  inheritedWidgetCount=0
  parentClass=/Game/Blueprints/Blutilities/ForTechArt/Widgets/Buttons/TA_BaseEditorButtonWidget.TA_BaseEditorButtonWidget_C
  rootWidgetClass=&Overlay
  widgetCount=7
 }
 widgets [
  Overlay_7 {widgetClassPath=&Overlay, widget=Overlay_7}
  VerticalBox_0 {
   widgetClassPath=&VerticalBox
   parent=Overlay_7
   widget=VerticalBox_0
   slot {class=OverlaySlot, properties {horizontalAlignment=HAlign_Fill}}
  }
  borderMain {
   widgetClassPath=&Border
   bIsVariable=true
   parent=VerticalBox_0
   widget=borderMain
   properties {padding=<ltrb 3 3 3 3>}
   slot {class=VerticalBoxSlot}
  }
  HorizontalBox_0 {
   widgetClassPath=&HorizontalBox
   parent=borderMain
   widget=HorizontalBox_0
   slot {class=BorderSlot, properties {padding=<ltrb 3 3 3 3>}}
  }
  ScaleBox_0 {
   widgetClassPath=&ScaleBox
   parent=HorizontalBox_0
   widget=ScaleBox_0
   properties {clipping=ClipToBounds, stretchDirection=DownOnly}
   slot {class=HorizontalBoxSlot, properties {padding=<ltrb 5 0 0 0>, size {sizeRule=Fill, value=1}}}
  }
  txtMain {
   widgetClassPath=&TextBlock
   bIsVariable=true
   parent=ScaleBox_0
   widget=txtMain
   properties {
    clipping=ClipToBounds
    justification=Center
    text=Text Block
    colorAndOpacity {colorUseRule=UseColor_Specified, specifiedColor=<rgba 0.0104 0.0104 0.0104 1>}
    font {
     bForceMonospaced=false
     bMaterialIsStencil=false
     fontMaterial=null
     fontObject=&Roboto
     letterSpacing=0
     monospacedWidth=1
     size=10
     skewAmount=0
     typefaceFontName=Bold
     outlineSettings {
      bApplyOutlineToDropShadows=false
      bMiteredCorners=false
      bSeparateFillAlpha=false
      outlineMaterial=null
      outlineSize=0
      outlineColor=<rgba 0 0 0 1>
     }
    }
    shadowColorAndOpacity=<rgba 0 0 0 0.2>
   }
   slot {class=ScaleBoxSlot, properties {horizontalAlignment=HAlign_Fill}}
  }
  TA_Generic01_Button {
   widgetClassPath=&TA_Generic01_Button
   parent=HorizontalBox_0
   widget=TA_Generic01_Button
   properties {
    buttonColor_Default=<rgba 1 0.1084 0 1>
    iconBrush {
     bIsDynamicallyLoaded=false
     drawAs=Image
     imageType=NoImage
     mirroring=NoMirror
     resourceName=None
     resourceObject=/Game/UI/Foundation/Textures/Buttons/T-Button-Clear-64
     tiling=NoTile
     imageSize=<8 8>
     margin=<ltrb 0 0 0 0>
     outlineSettings {
      bUseBrushTransparency=false
      roundingType=HalfHeightRadius
      width=0
      color {colorUseRule=UseColor_Specified, specifiedColor=<rgba 0 0 0 0>}
      cornerRadii=<0 0 0 1>
     }
     tintColor {colorUseRule=UseColor_Specified, specifiedColor=<rgba 1 1 1 1>}
     uVRegion {bIsValid=false, max=<0 0>, min=<0 0>}
    }
   }
   slot {class=HorizontalBoxSlot, properties {horizontalAlignment=HAlign_Right, verticalAlignment=VAlign_Center, padding=<ltrb 5 0 0 0>}}
  }
 ]
}
